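# Niagara system that renders orange impact effects for the Dark Voyager shield, driven by a Niagara Data Channel.
# Visual emitters spawn from channel impacts: energy ring, runes, sparks, fat streaks, and a mesh-and-sprite spiky flash.
# Each initializes particles from impact position, normal and scale, which set orientation, size, lifetime or speed.
# An enabled Debug emitter bursts writes into the shield impact channel using User.Scale; a Minimal test writer is disabled.
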NiagaraSystem /DragonCartNarrative/Assets/VFX/OrbBarrier_Impact/NS_NDC_DarkVoyager_Shield_Impact
dependencies {
 usedDataInterfaces [
  &NiagaraDataInterfaceCurve
  &NiagaraDataInterfaceVector2DCurve
  /Script/Niagara.NiagaraDataInterfaceDataChannelWrite
  &NiagaraDataInterfaceSpriteRendererInfo
  /Script/Niagara.NiagaraDataInterfaceDataChannelRead
 ]
 usedDynamicInputs []
 usedModules [
  &SystemState
  &self:SystemSpawnScript.SystemScriptSource.SystemScriptGraph.NiagaraNodeAssignment_1.SetVariables_#1
  /Niagara/Modules/System/CompleteIfUnused
 ]
}
emitters {
 Debug {
  emitterUpdateScript [
   &EmitterState(Life Cycle Mode=::NewEnumerator1, Loop Duration=2, UseLoopDelay=false)
   &SpawnBurst_Instantaneous(Use Spawn Probability=false, UseLoopCountLimit=false)
  ]
  particleSpawnScript [
   &InitializeParticle(Color Mode=::NewEnumerator0, Lifetime=0.5, UsePositionOffset=false, Write Material Random=true)
   &self:ScratchModule(
    Data Channel Writer=DataInterface() {propertyValues {
      AllocationCount=1
      AllocationMode=Static
      Channel=/DragonCartNarrative/Assets/VFX/OrbBarrier/OrbBarrierImpact/NDC_DCN_DarkVoyager_Shield_Impact
      bOnlyWriteOnceOnSubticks=true
      bPublishToCPU=true
      bPublishToGPU=true
      bPublishToGame=false
      bUpdateDestinationDataEveryTick=true
      AccessContext {accessContext {}}
     }}
    Scale:NiagaraFloat=$User.Scale
   ) {module=ScratchModule}
  ]
  particleUpdateScript [&ParticleState()]
  rendererClasses [&NiagaraSpriteRendererProperties]
  renderers [{rendererClass=&NiagaraSpriteRendererProperties}]
 }
 EnergyRing {
  emitterUpdateScript [
   &EmitterState()
   /Game/Effects/Niagara/DataChannels/Impacts/Spawn_NDCFortImpacts()
  ]
  particleSpawnScript [
   &self:Initialize_Particle_from_NDC() {module=Initialize_Particle_from_NDC}
   &InitializeParticle(
    Color=<rgba 5 1.4112 0 1>
    Lifetime=&Lerp_Float(
     A=&RandomRangeFloat(Minimum=0.14, Maximum=0.17)
     B=&RandomRangeFloat(Minimum=0.18, Maximum=0.24)
     Alpha=$Particles.ImpactScale
    )
    Position=$Particles.Position
    Position Mode=::NewEnumerator1
    Sprite Rotation Mode=::NewEnumerator2
    Sprite Size Mode=::NewEnumerator3
    Uniform Sprite Size=&Multiply_Float(A=1200, B=&Lerp_Float(A=0.25, B=1, Alpha=$Particles.ImpactScale))
    UsePositionOffset=false
    Write Material Random=true
   )
   &CameraOffset(Camera Offset Amount=50)
   &SpriteFacingAndAlignment(Sprite Facing=$Particles.ImpactSurfaceNormal, WriteAlignment=false, WriteFacing=true)
  ]
  particleUpdateScript [
   &ParticleState()
   &ScaleSpriteSize(
    Uniform Curve Scale=0.25
    Uniform Curve Sprite Scale=DataInterface() {propertyValues {Curve {keys [
        {arriveTangent=6.3514, interpMode=RCIM_Cubic, leaveTangent=6.3514, tangentMode=RCTM_User, time=0, value=0}
        {tangentMode=RCTM_User, time=1, value=1}
       ]}}}
   )
   &SolveForcesAndVelocity(Clamp Velocity=false, Limit Acceleration=false)
   &ScaleColor(
    Scale Alpha=&FloatFromCurve(FloatCurve=DataInterface() {propertyValues {Curve {keys [<at 0.4 1 RCIM_Cubic>, {arriveTangent=-2.2027, tangentMode=RCTM_User, time=1, value=0}]}}})
    Scale RGB=&VectorFromFloat(Value=1)
    ScaleA=true
    ScaleRGB=true
   )
   /SetVariables(Particles.Glow Sprite Color:LinearColor=&MakeLinearColorFromVectorAndFloat(
     "Vector (RGB)"=&MakeVectorFromLinearColorRGB(Color=<rgba 5 1.4112 0 1>)
     "Float (Alpha)"=&FloatFromCurve(
      FloatCurve=DataInterface() {propertyValues {Curve {keys [
          {arriveTangent=-2.7319, interpMode=RCIM_Cubic, time=0.05, value=1}
          {arriveTangent=-1.5377, interpMode=RCIM_Cubic, leaveTangent=-1.5377, tangentMode=RCTM_User, time=0.2, value=0.55}
          <at 1 0>
         ]}}}
      Scale Curve=1.3
     )
    )) {isSetParameters=true, module=SetVariables_#2}
  ]
  rendererClasses [&NiagaraSpriteRendererProperties]
  renderers [{rendererClass=&NiagaraSpriteRendererProperties}]
 }
 Minimal {
  bEnabled=false
  emitterUpdateScript [
   &EmitterState()
   &SpawnBurst_Instantaneous(Use Spawn Probability=false, UseLoopCountLimit=false)
  ]
  particleSpawnScript [
   &InitializeParticle(Color Mode=::NewEnumerator0, UsePositionOffset=false, Write Material Random=true)
   &self:ScratchModule(
    Data Channel Writer=DataInterface() {propertyValues {
      AllocationCount=1
      AllocationMode=Static
      Channel=&NDC_DCN_DarkVoyager_Shield_Impact
      bOnlyWriteOnceOnSubticks=true
      bPublishToCPU=true
      bPublishToGPU=true
      bPublishToGame=false
      bUpdateDestinationDataEveryTick=true
      AccessContext {accessContext {}}
     }}
    Scale:NiagaraFloat=$User.Scale
   ) {module=ScratchModule}
   &self:Data_Channel_Debug(
    Data Channel Writer=DataInterface() {propertyValues {
      AllocationCount=1
      AllocationMode=Static
      Channel=&NDC_DCN_DarkVoyager_Shield_Impact
      bOnlyWriteOnceOnSubticks=true
      bPublishToCPU=true
      bPublishToGPU=true
      bPublishToGame=true
      bUpdateDestinationDataEveryTick=true
      AccessContext {accessContext {}}
     }}
    ImpactNormal:Vector3f=$Engine.Owner.SystemXAxis
    ImpactPoint:NiagaraPosition=/Niagara/DynamicInputs/Helpers/SimulationPosition()
    ScaleMultiplier=1
   ) {module=Data_Channel_Debug}
  ]
  particleUpdateScript [&ParticleState()]
  rendererClasses [&NiagaraMeshRendererProperties]
  renderers [{rendererClass=&NiagaraMeshRendererProperties, rendererIndex=0}]
 }
 Runes {
  emitterUpdateScript [
   &EmitterState()
   &Spawn_NDCFortImpacts(Spawn Count=6)
  ]
  particleSpawnScript [
   &self:Initialize_Particle_from_NDC() {module=Initialize_Particle_from_NDC}
   &InitializeParticle(
    Color=<rgba 5 1.4112 0 1>
    Lifetime=&Lerp_Float(
     A=&RandomRangeFloat(Minimum=0.11, Maximum=0.4)
     B=&RandomRangeFloat(Minimum=0.25, Maximum=0.5)
     Alpha=$Particles.ImpactScale
    )
    Position=$Particles.Position
    Position Mode=::NewEnumerator1
    Sprite Size Mode=::NewEnumerator3
    Uniform Sprite Size=&Multiply_Float(A=120, B=&Lerp_Float(A=0.25, B=1, Alpha=$Particles.ImpactScale))
    UsePositionOffset=false
    Write Material Random=true
   )
   &Color(Color=&MakeLinearColorFromVectorAndFloat(
     "Vector (RGB)"=<4 0.8 0.1>
     "Float (Alpha)"=&RandomRangeFloat(Minimum=0.2, Maximum=0.4, Evaluation Type=::NewEnumerator1)
    ))
   &ShapeLocation(
    Cylinder Height=5
    Cylinder Radius=60
    MaskSpawnGroup=false
    Override Seed=false
    Rotation Mode=::NewEnumerator2
    Shape Primitive=::NewEnumerator1
    Surface Only=false
    UseOwnerScale=true
    Yaw / Pitch / Roll=<1 0 0>
   )
   &CameraOffset(Camera Offset Amount=50) {enabled=false}
   /CascadeToNiagaraConverter/NiagaraScripts/ModuleScripts/AddVectorToPosition(Vector:Vector3f=&TransformVector(Vector=<4 0 0>, Destination Space=::Simulation))
   &DynamicMaterialParameters(
    Index 0 Param 1=&RandomRangeFloat(Minimum=0, Maximum=1, Evaluation Type=::NewEnumerator1)
    Index 0 Param 2=&RandomRangeFloat(Minimum=1, Maximum=12, Evaluation Type=::NewEnumerator1)
    Module.Param0WriteEnabled=true
    Module.Param1WriteEnabled=true
    Module.Param2WriteEnabled=false
    Module.Param3WriteEnabled=false
   )
   &SpriteFacingAndAlignment(Sprite Facing=$Particles.ImpactSurfaceNormal, WriteAlignment=false, WriteFacing=true)
  ]
  particleUpdateScript [
   &ParticleState()
   &ScaleColor(
    Scale Alpha=&FloatFromCurve(FloatCurve=DataInterface() {propertyValues {Curve {keys [{arriveTangent=-2.2945, interpMode=RCIM_Cubic, leaveTangent=-2.2945, tangentMode=RCTM_User, time=0, value=1}, <at 1 0>]}}})
    Scale RGB=&VectorFromFloat(Value=&RandomRangeFloat(Minimum=8, Maximum=0.5, Evaluation Type=::NewEnumerator1))
    ScaleA=true
    ScaleRGB=true
   )
   &ScaleSpriteSize(Uniform Curve Sprite Scale=DataInterface() {propertyValues {Curve {keys [
        {arriveTangent=6.3514, interpMode=RCIM_Cubic, leaveTangent=6.3514, tangentMode=RCTM_User, time=0, value=0}
        {tangentMode=RCTM_User, time=1, value=1}
       ]}}}) {enabled=false}
   &SolveForcesAndVelocity(Clamp Velocity=false, Limit Acceleration=false)
   &SubUVAnimation(
    End Frame Range Override=2
    "Sprite Renderer "=DataInterface() {propertyValues {SpriteRenderer=&self:Runes_1.NiagaraSpriteRendererProperties_0}}
    SubUV Animation Mode:ENiagara_SubUVLookupModeV2=::NewEnumerator2
    UseEndFrame=false
    UseStartFrame=false
   )
   /SetVariables(Particles.Glow Sprite Color:LinearColor=&MakeLinearColorFromVectorAndFloat(
     "Vector (RGB)"=&MakeVectorFromLinearColorRGB(Color=<rgba 5 1.4112 0 1>)
     "Float (Alpha)"=&FloatFromCurve(
      FloatCurve=DataInterface() {propertyValues {Curve {keys [
          {arriveTangent=-2.7319, interpMode=RCIM_Cubic, time=0.05, value=1}
          {arriveTangent=-1.5377, interpMode=RCIM_Cubic, leaveTangent=-1.5377, tangentMode=RCTM_User, time=0.2, value=0.55}
          <at 1 0>
         ]}}}
      Scale Curve=0.4
     )
    )) {isSetParameters=true, module=SetVariables_#2}
  ]
  rendererClasses [&NiagaraSpriteRendererProperties]
  renderers [{rendererClass=&NiagaraSpriteRendererProperties}]
 }
 Sparks {
  emitterUpdateScript [
   &EmitterState()
   &Spawn_NDCFortImpacts(Max Spawn Count=10, Min Spawn Count=5, Spawn Count Mode=2)
  ]
  particleSpawnScript [
   &self:Initialize_Particle_from_NDC() {module=Initialize_Particle_from_NDC}
   &InitializeParticle(
    Color Maximum=<rgba 5 1.4112 0 1>
    Color Minimum=<rgba 10 0.4672 0.0018 0.4>
    Color Mode=::NewEnumerator2
    Lifetime=&FloatFromCurve(
     FloatCurve=DataInterface() {propertyValues {Curve {
        defaultValue=1
        keys [
         {arriveTangent=0.4191, interpMode=RCIM_Cubic, leaveTangent=0.4191, tangentMode=RCTM_User, time=0, value=0.1}
         {arriveTangent=2.3656, interpMode=RCIM_Cubic, leaveTangent=2.3656, tangentMode=RCTM_User, time=1, value=1}
        ]
       }}}
     CurveIndex=&RandomRangeFloat(Minimum=0, Maximum=0.25)
    )
    Position=$Particles.Position
    Position Mode=::NewEnumerator1
    Sprite Size Mode=::NewEnumerator1
    Uniform Sprite Size Max=2
    Uniform Sprite Size Min=1
    UsePositionOffset=false
    Write Material Random=true
   )
   &ShapeLocation(MaskSpawnGroup=false, Override Seed=false, Sphere Radius=10, UseOwnerScale=true)
   &AddVelocity(
    Cone Angle=110
    Cone Axis=$Particles.ImpactSurfaceNormal
    Inner Cone Angle=45
    Module.UseConeFalloff=true
    Module.UseVelDistribution=false
    Override Seed=false
    Speed Falloff From Cone Axis=0.2
    Velocity Mode=::NewEnumerator2
    Velocity Speed=&Lerp_Float(
     A=300
     B=1000
     Alpha=&FloatFromCurve(
      FloatCurve=DataInterface() {propertyValues {Curve {keys [
          {arriveTangent=0.2237, interpMode=RCIM_Cubic, leaveTangent=0.2237, tangentMode=RCTM_User, time=0, value=0}
          {arriveTangent=2.6808, interpMode=RCIM_Cubic, leaveTangent=2.6808, tangentMode=RCTM_User, time=1, value=1}
         ]}}}
      CurveIndex=&RandomRangeFloat(Minimum=0, Maximum=1)
      Scale Curve=&Lerp_Float(A=0.5, B=1, Alpha=$Particles.ImpactScale)
     )
    )
   )
  ]
  particleUpdateScript [
   &ParticleState()
   &GravityForce(Gravity=<0 0 -100>)
   &Drag(
    Drag=&FloatFromCurve(
     FloatCurve=DataInterface() {propertyValues {Curve {keys [
         {interpMode=RCIM_Cubic, leaveTangent=-8.9049, leaveTangentWeight=0.5018, tangentMode=RCTM_Break, time=0, value=1}
         {interpMode=RCIM_Cubic, tangentMode=RCTM_Break, time=0.3, value=0.025}
        ]}}}
     Scale Curve=&RandomRangeFloat(Minimum=5, Maximum=12)
    )
    Use Linear Drag=true
    Use Rotational Drag=false
   )
   &SolveForcesAndVelocity(Clamp Velocity=false, Limit Acceleration=false)
   &ScaleSpriteSize(Uniform Curve Sprite Scale=DataInterface() {propertyValues {Curve {keys [
        {interpMode=RCIM_Cubic, tangentMode=RCTM_Break, time=0.25, value=1}
        {arriveTangent=-2.298, interpMode=RCIM_Cubic, tangentMode=RCTM_Break, time=1, value=0.2}
       ]}}})
   &ScaleSpriteSizeBySpeed(
    Max Scale Factor=<1 10>
    Sample Scale Factor By Curve=true
    Scale Factor Curve=DataInterface() {propertyValues {Curve {keys [{arriveTangent=0.3851, interpMode=RCIM_Cubic, time=0, value=0}, {arriveTangent=2.2117, tangentMode=RCTM_User, time=1, value=1}]}}}
    Velocity Threshold=2000
   )
   &ScaleColor(
    Scale Alpha=&FloatFromCurve(FloatCurve=DataInterface() {propertyValues {Curve {keys [<at 0 1 RCIM_Cubic>, {arriveTangent=-2.9021, tangentMode=RCTM_User, time=1, value=0}]}}})
    ScaleA=true
    ScaleRGB=false
   ) {module=ScaleColor001}
   &DynamicMaterialParameters(
    Index 0 Param 1=&FloatFromCurve(
     FloatCurve=DataInterface() {propertyValues {Curve {defaultValue=1, keys [<at 0 0 RCIM_Cubic>, {arriveTangent=1.1412, tangentMode=RCTM_User, time=1, value=1}]}}}
     Scale Curve=2
    )
    Index 0 Param 3=3
    Module.Index0Float0Write=true
    Module.Index0Float1Write=false
    Module.Index0Float2Write=true
    Module.Index0Float3Write=false
   )
  ]
  rendererClasses [&NiagaraSpriteRendererProperties]
  renderers [{rendererClass=&NiagaraSpriteRendererProperties}]
 }
 SpikeyFlash {
  emitterUpdateScript [
   &EmitterState()
   &Spawn_NDCFortImpacts()
  ]
  particleSpawnScript [
   &self:Initialize_Particle_from_NDC() {module=Initialize_Particle_from_NDC}
   &InitializeParticle(
    Color=<rgba 5 1.4112 0 1>
    Lifetime=&Lerp_Float(
     A=&RandomRangeFloat(Minimum=0.14, Maximum=0.17)
     B=&RandomRangeFloat(Minimum=0.18, Maximum=0.24)
     Alpha=$Particles.ImpactScale
    )
    Mesh Scale=&Multiply_VectorByFloat(
     Vector=&RandomRangeVector(Minimum=<0.05 0.05 0.25>, Maximum=<0.07 0.07 0.5>)
     Float=&Lerp_Float(A=0.5, B=1, Alpha=$Particles.ImpactScale)
    )
    Mesh Scale Mode=::NewEnumerator4
    Position=$Particles.Position
    Position Mode=::NewEnumerator1
    Sprite Rotation Mode=::NewEnumerator2
    Sprite Size Mode=::NewEnumerator3
    Uniform Sprite Size=&Multiply_Float(A=850, B=&Lerp_Float(A=0.25, B=1, Alpha=$Particles.ImpactScale))
    UsePositionOffset=false
    Write Material Random=true
   )
   &CameraOffset(Camera Offset Amount=50)
   &InitialMeshOrientation(
    Mesh Axis To Orient=<0 0 1>
    Mesh Orientation Mode=::NewEnumerator2
    Orientation Vector=$Particles.ImpactSurfaceNormal
    Rotation=&RandomRangeVector(Minimum=<0 0 0>, Maximum=<0 0 1>)
    Use Rotation Vector=true
   )
   &SpriteFacingAndAlignment(Sprite Facing=$Particles.ImpactSurfaceNormal, WriteAlignment=false, WriteFacing=true)
  ]
  particleUpdateScript [
   &ParticleState()
   &ScaleSpriteSize(
    Uniform Curve Scale=0.3
    Uniform Curve Sprite Scale=DataInterface() {propertyValues {Curve {keys [
        {arriveTangent=6.3514, interpMode=RCIM_Cubic, leaveTangent=6.3514, tangentMode=RCTM_User, time=0, value=0}
        {tangentMode=RCTM_User, time=1, value=1}
       ]}}}
   )
   &SolveForcesAndVelocity(Clamp Velocity=false, Limit Acceleration=false)
   &ScaleMeshSize(Scale Factor=&VectorFromFloat(Value=&FloatFromCurve(
      FloatCurve=DataInterface() {propertyValues {Curve {keys [
          {arriveTangent=9.901, interpMode=RCIM_Cubic, leaveTangent=9.901, tangentMode=RCTM_User, time=0, value=0.1}
          {arriveTangent=2.6083, interpMode=RCIM_Cubic, leaveTangent=2.6083, tangentMode=RCTM_User, time=1, value=6}
         ]}}}
      Scale Curve=0.3
     )))
   &ScaleColor(
    Scale Alpha=&FloatFromCurve(FloatCurve=DataInterface() {propertyValues {Curve {keys [<at 0.4 1 RCIM_Cubic>, {arriveTangent=-2.2027, tangentMode=RCTM_User, time=1, value=0}]}}})
    Scale RGB=&VectorFromFloat(Value=2)
    ScaleA=true
    ScaleRGB=true
   )
   /SetVariables(Particles.Glow Sprite Color:LinearColor=&MakeLinearColorFromVectorAndFloat(
     "Vector (RGB)"=&MakeVectorFromLinearColorRGB(Color=<rgba 5 1.4112 0 1>)
     "Float (Alpha)"=&FloatFromCurve(
      FloatCurve=DataInterface() {propertyValues {Curve {keys [
          {arriveTangent=-2.7319, interpMode=RCIM_Cubic, time=0.05, value=1}
          {arriveTangent=-1.5377, interpMode=RCIM_Cubic, leaveTangent=-1.5377, tangentMode=RCTM_User, time=0.2, value=0.55}
          <at 1 0>
         ]}}}
      Scale Curve=5
     )
    )) {isSetParameters=true, module=SetVariables_#2}
  ]
  rendererClasses [&NiagaraMeshRendererProperties, &NiagaraSpriteRendererProperties]
  renderers [
   {rendererClass=&NiagaraMeshRendererProperties, rendererIndex=0}
   {rendererClass=&NiagaraSpriteRendererProperties, rendererIndex=1}
  ]
 }
 Streaks_Fat {
  emitterUpdateScript [
   &EmitterState()
   &Spawn_NDCFortImpacts(Max Spawn Count=5, Min Spawn Count=3, Spawn Count Mode=2)
  ]
  particleSpawnScript [
   &self:Initialize_Particle_from_NDC() {module=Initialize_Particle_from_NDC}
   &InitializeParticle(
    Color=<rgba 5 1.4112 0 1>
    Lifetime Max=0.15
    Lifetime Min=0.11
    Lifetime Mode=::NewEnumerator1
    Position=$Particles.Position
    Position Mode=::NewEnumerator1
    Sprite Size=&Multiply_Vector2D_ByFloat(
     Float=&Lerp_Float(A=0.15, B=0.3, Alpha=$Particles.ImpactScale)
     Vector2D=&RandomRangeVector2D(Minimum=<10 100>, Maximum=<15 150>)
    )
    Sprite Size Mode=::NewEnumerator4
    UsePositionOffset=false
    Write Material Random=true
   )
   &ShapeLocation(MaskSpawnGroup=false, Override Seed=false, Sphere Radius=1, Sphere Surface Distribution=0.5595, UseOwnerScale=true)
   &AddVelocity(
    Cone Angle=10
    Cone Axis=$Particles.ImpactSurfaceNormal
    Inner Cone Angle=180
    Module.UseConeFalloff=false
    Module.UseVelDistribution=false
    Override Seed=false
    Velocity Mode=::NewEnumerator2
    Velocity Speed=&Lerp_Float(
     A=0
     B=1
     Alpha=&FloatFromCurve(
      FloatCurve=DataInterface() {propertyValues {Curve {keys [
          {interpMode=RCIM_Cubic, tangentMode=RCTM_User, time=0, value=0}
          {arriveTangent=2.6808, interpMode=RCIM_Cubic, leaveTangent=2.6808, tangentMode=RCTM_User, time=1, value=1}
         ]}}}
      CurveIndex=&RandomRangeFloat(Minimum=0, Maximum=1)
     )
    )
   )
  ]
  particleUpdateScript [
   &ParticleState()
   &SolveForcesAndVelocity(Clamp Velocity=false, Limit Acceleration=false)
   &ScaleSpriteSize(Scale Factor=&Vector2DFromCurve(
     Vector2Curve=DataInterface() {propertyValues {XCurve {keys [<at 0 0.75 RCIM_Cubic>,<at 1 1 RCIM_Cubic>]}, YCurve {keys [<at 0 0 RCIM_Cubic>,<at 1 2 RCIM_Cubic>]}}}
     ScaleCurve=<1 1>
    ))
   &ScaleColor(
    Scale Alpha=&FloatFromCurve(FloatCurve=DataInterface() {propertyValues {Curve {keys [<at 0 1 RCIM_Cubic>,<at 1 0>]}}})
    Scale RGB=&VectorFromFloat(Value=&FloatFromCurve(FloatCurve=DataInterface() {propertyValues {Curve {keys [<at 0 10 RCIM_Cubic>,<at 1 1 RCIM_Cubic>]}}}))
    ScaleA=true
    ScaleRGB=true
   )
  ]
  rendererClasses [&NiagaraSpriteRendererProperties]
  renderers [{rendererClass=&NiagaraSpriteRendererProperties}]
 }
}
summary {
 systemName=NS_NDC_DarkVoyager_Shield_Impact
 emitters [
  {emitterName=Sparks, rendererClasses [&NiagaraSpriteRendererProperties]}
  {emitterName=Streaks_Fat, rendererClasses [&NiagaraSpriteRendererProperties]}
  {bEnabled=false, emitterName=Minimal, rendererClasses [&NiagaraMeshRendererProperties]}
  {emitterName=SpikeyFlash, rendererClasses [&NiagaraMeshRendererProperties, &NiagaraSpriteRendererProperties]}
  {emitterName=Debug, rendererClasses [&NiagaraSpriteRendererProperties]}
  {emitterName=EnergyRing, rendererClasses [&NiagaraSpriteRendererProperties]}
  {emitterName=Runes, rendererClasses [&NiagaraSpriteRendererProperties]}
 ]
 userVariables [
  User.Lifetime {type=NiagaraFloat, defaultValue=20, description=""}
  User.Scale {type=NiagaraFloat, defaultValue=0, description=""}
 ]
}
